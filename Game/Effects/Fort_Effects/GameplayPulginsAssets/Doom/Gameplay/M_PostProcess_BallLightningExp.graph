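# Translucent, depth-test-free unlit material for Niagara mesh particles, drawing a screen-space ball lightning explosion overlay, judging by its name.
# Viewport UVs become radial coordinates for panning plasma, distorted by noise, then refined and tinted.
# Inverted screen-center sphere masks concentrate glow and energy toward screen edges.
# Opacity fades with camera distance from a point built from dynamic Param2, relative to dynamic Scale, computed per vertex.
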Material /Game/Effects/Fort_Effects/GameplayPulginsAssets/Doom/Gameplay/M_PostProcess_BallLightningExp
graph {
 properties {
  bDisableDepthTest=true
  bUseTranslucencyVertexFog=false
  bUsedWithNiagaraMeshParticles=true
  blendMode=BLEND_Translucent
  pixelDepthOffsetMode=PDOM_Legacy
  shadingModel=MSM_Unlit
  twoSided=true
  usedShadingModels=Unlit
 }
 nodes [
  ActorPositionWS()
  n1 = TextureObjectParameter() {
   group=Energy
   parameterName=Energy_Texture
   samplerType=SAMPLERTYPE_LinearColor
   sortPriority=0
   texture=/Game/Effects/Fort_Effects/Emotes/UltraLight/Texture/T_Plasma_A_C
  }
  n2 = TextureObjectParameter() {
   group=NoiseForPost
   parameterName=NoiseForPost_Texture
   samplerType=SAMPLERTYPE_Masks
   sortPriority=0
   texture=/Game/Global/Textures/Default/Noise/T_Metal_Noise
  }
  n3 = VectorParameter() {group=NoiseForPost, parameterName=NoiseForPost_UVControl, channelNames {a=VPan, b=UPan, g=VScale, r=UScale}, defaultValue=<rgba 1 0.02 0 -8>}
  n4 = ScreenPosition()
  n5 = MaterialFunctionCall(Vector or UVs=Subtract(A=Multiply(A=@n4.ViewportUV) {constB=2}, B=Constant() {r=1})) {materialFunction=&VectorToRadialValue}
  n6 = VectorParameter() {
   group=NoiseForPost
   parameterName=NoiseForPost_NoiseControl
   sortPriority=60
   channelNames {a=VStrength, b=UStrength, g=VOffset, r=UOffset}
   defaultValue=<rgba -0.5 -0.5 0 0>
  }
  n7 = VectorParameter() {group=Energy, parameterName=Energy_UVControl, channelNames {a=VPan, b=UPan, g=VScale, r=UScale}, defaultValue=<rgba 2 0.1 0 -6>}
  n8 = MaterialFunctionCall(Float2=@n5.Radial Coordinates) {materialFunction=&BreakOutFloat2Components}
  n9 = MaterialFunctionCall(
   Texture=@n1
   AdditionalUVOffset=MaterialFunctionCall!(
    NoiseValue=MaterialFunctionCall(Texture=@n2, UVControl_RGB=@n3, UVControl_A=@n3, "[Override]UVInput"=@n5.Radial Coordinates) {materialFunction=&FXCommon_TextureSample}.RG
    Control_RGB=@n6
    Control_A=@n6
   ) {materialFunction=/Game/Effects/Niagara/MasterMaterials/Functions/Common/FXCommon_NoiseControl}.NoiseValue
   UVControl_RGB=@n7
   UVControl_A=@n7
   "[Override]UVInput"=MaterialFunctionCall(X=@n8.R, Y=Power(Base=@n8.G)) {materialFunction=&MakeFloat2}.Result
  ) {materialFunction=&FXCommon_TextureSample}
  n10 = VectorParameter() {parameterName=Energy_Refine, defaultValue=<rgba 5 0 50 0.5>}
  n11 = MaterialFunctionCall(Texture=@n9.R, RGB=@n10.RGB, A=@n10.RGB) {materialFunction=&MF_FX_Refine}
  n12 = Constant2Vector() {g=0.5, r=0.5}
  n13 = Constant() {r=1}
  LinearInterpolate(
   A=@n11.Result
   B=ScalarParameter() {defaultValue=0.5, parameterName=B}
   Alpha=SphereMask(A=@n4.ViewportUV, B=@n12, Radius=@n13, Hardness=ScalarParameter() {defaultValue=1, parameterName=HDNS2}) {hardnessPercent=1}
  )
  LinearInterpolate()
  n14 = VectorParameter() {parameterName=Energy_RefineToColor, defaultValue=<rgba 4 2 0.2 0.5>}
  n15 = Multiply(
   A=VectorParameter() {parameterName=Color, defaultValue=<rgba 1.3333 4 0 1>}.RGB
   B=MaterialFunctionCall(Texture=@n9.R, RGB=@n14.RGB, A=@n14.RGB) {materialFunction=&MF_FX_Refine}.Result
  )
  n16 = VectorParameter() {parameterName=GlowColor, defaultValue=<rgba 2.55 3 0.3 1>}
  n17 = SphereMask(A=@n4.ViewportUV, B=@n12, Radius=@n13) {hardnessPercent=0}
  n18 = Multiply(
   A=Power(Base=OneMinus(None=@n17), Exp=ScalarParameter() {defaultValue=4, parameterName=EdgePower})
   B=ScalarParameter() {defaultValue=2, parameterName=EdgeGlow}
  )
  MaterialFunctionCall(Emissive=Multiply(A=LinearInterpolate(A=@n15, B=@n16.RGB, Alpha=@n18), B=ParticleColor().RGB)) {materialFunction=&MF_FX_EmissiveAdjustments}
  n19 = SphereMask(A=@n4.ViewportUV, B=@n12, Radius=@n13, Hardness=ScalarParameter() {defaultValue=0.2, parameterName=HDNS})
  n20 = Multiply(A=Power(Base=OneMinus(None=@n19), Exp=ScalarParameter() {defaultValue=3, parameterName=EXP}) {constExponent=1.5}, B=@n11.Result)
  n21 = CameraPositionWS()
  n22 = DynamicParameter() {paramNames [Scale,Param2,Param3,Param4]}
  n23 = MaterialFunctionCall(X=@n22.Param2, Y=@n22.Param2, Z=@n22.Param2) {materialFunction=&MakeFloat3}
  n24 = Divide(
   A=Length(None=Subtract(A=@n21, B=@n23.Result)) {is=class MaterialExpressionLength}
   B=Multiply(A=@n22.Scale, B=ScalarParameter() {defaultValue=50, parameterName=BaseScaleSPhere}) {constB=50}
  )
  n25 = ScalarParameter() {defaultValue=0.4, parameterName=BlendWidth}
  n26 = Saturate(None=Subtract(A=Saturate(None=@n24), B=@n25) {constB=20})
  n27 = Divide(B=OneMinus(None=@n25))
  n28 = Power(Base=OneMinus(None=Multiply(A=@n26, B=@n27))) {constExponent=1.5}
  Multiply(A=LinearInterpolate(B=Add(A=@n18, B=@n20), Alpha=VertexInterpolator(VS=@n28).PS), B=ParticleColor().A)
 ]
}
