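# Translucent unlit screen-effect material for Niagara mesh particles, apparently the Doom ball lightning overlay, with depth test disabled.
# Radial screen UVs drive two noise layers that distort a tinted rainbow color texture and a scrolling gradient scan mask, steered by dynamic parameters.
# Edge and sphere masks shape center glow and opacity.
# Opacity fades by camera distance, computed per vertex, and view direction toward a point built from dynamic parameter X alone.
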Material /Game/Effects/Fort_Effects/GameplayPulginsAssets/Doom/Gameplay/M_PostProcess_BallLightningExp2
graph {
 properties {
  bDisableDepthTest=true
  bUseTranslucencyVertexFog=false
  bUsedWithNiagaraMeshParticles=true
  blendMode=BLEND_Translucent
  pixelDepthOffsetMode=PDOM_Legacy
  shadingModel=MSM_Unlit
  twoSided=true
  usedShadingModels=Unlit
 }
 nodes [
  ActorPositionWS()
  n1 = TextureObjectParameter() {group=Color, parameterName=Color_Texture, sortPriority=0, texture=&T_Tile_RainbowColor}
  n2 = DynamicParameter() {parameterIndex=1, paramNames [Pan,NoiseStr,ScaleV,Param4]}
  n3 = TextureObjectParameter() {group=NoiseB, parameterName=NoiseB_Texture, sortPriority=0, texture=&T_Inky_Smoke_Tile}
  n4 = VectorParameter() {group=NoiseB, parameterName=NoiseB_UVControl, channelNames {a=VPan, b=UPan, g=VScale, r=UScale}, defaultValue=<rgba 3 0.5 0 1>}
  n5 = ScreenPosition()
  n6 = MaterialFunctionCall(Vector or UVs=Subtract(A=Multiply(A=@n5.ViewportUV) {constB=2})) {materialFunction=&VectorToRadialValue}
  n7 = MaterialFunctionCall(Texture=@n3, UVControl_RGB=@n4, UVControl_A=@n4, "[Override]UVInput"=@n6.Radial Coordinates) {materialFunction=&FXCommon_TextureSample}
  n8 = VectorParameter() {
   group=NoiseB
   parameterName=NoiseB_NoiseControl
   sortPriority=60
   channelNames {a=VStrength, b=UStrength, g=VOffset, r=UOffset}
   defaultValue=<rgba -0.5 -0.5 0 1>
  }
  n9 = VectorParameter() {group=NosieA, parameterName=NosieA_UVControl, channelNames {a=VPan, b=UPan, g=VScale, r=UScale}, defaultValue=<rgba 4 0.7 0 1>}
  n10 = MaterialFunctionCall(
   Texture=TextureObjectParameter() {group=NosieA, parameterName=NosieA_Texture, sortPriority=0, texture=/Game/Items/Traps/Textures/T_PerlinRGB_01_M}
   UVControl_RGB=@n9
   UVControl_A=@n9
   "[Override]UVInput"=@n6.Radial Coordinates
  ) {materialFunction=&FXCommon_TextureSample}
  n11 = VectorParameter() {
   group=NosieA
   parameterName=NosieA_NoiseControl
   sortPriority=60
   channelNames {a=VStrength, b=UStrength, g=VOffset, r=UOffset}
   defaultValue=<rgba -0.5 -0.5 0 1>
  }
  n12 = Multiply(
   A=@n2.NoiseStr
   B=Add(
    A=MaterialFunctionCall!(NoiseValue=@n7.R, Control_RGB=@n8, Control_A=@n8) {materialFunction=/Game/Effects/Niagara/MasterMaterials/Functions/Common/FXCommon_NoiseControl}.NoiseValue
    B=MaterialFunctionCall(NoiseValue=@n10.RG, Control_RGB=@n11, Control_A=@n11) {materialFunction=&FXCommon_NoiseControl}.NoiseValue
   )
  )
  n13 = Multiply(A=@n12, B=ScalarParameter() {defaultValue=0.1, parameterName=123143t})
  n14 = VectorParameter() {group=Color, parameterName=Color_UVControl, channelNames {a=VPan, b=UPan, g=VScale, r=UScale}, defaultValue=<rgba 1 0.5 0 0.5>}
  n15 = VectorParameter() {parameterName=B, defaultValue=<rgba 5 3.0208 0.25 1>}
  n16 = Multiply(
   A=Desaturation(
    None=MaterialFunctionCall(Texture=@n1, AdditionalUVOffset=@n13, UVControl_RGB=@n14, UVControl_A=@n14, "[Override]UVInput"=@n6.Radial Coordinates) {materialFunction=&FXCommon_TextureSample}.RGB
    Fraction=ScalarParameter() {defaultValue=0.5, parameterName=231}
   )
   B=@n15.RGB
  )
  n17 = MaterialFunctionCall(Float2=@n6.Radial Coordinates) {materialFunction=&BreakOutFloat2Components}
  n18 = VectorParameter() {
   group=CenterBright
   parameterName=CenterBright_EdgeMask_InvertControl
   channelNames {a="", b=Power, g=Multiply, r=Offset}
   defaultValue=<rgba 0 1 2 1>
  }
  n19 = VectorParameter() {
   group=CenterBright
   parameterName=CenterBright_EdgeMask_Control
   channelNames {a="", b=Power, g=Multiply, r=Offset}
   defaultValue=<rgba 1 2 3 1>
  }
  n20 = MaterialFunctionCall!(UorV=@n17.G, Invert_EdgeControl=@n18, NoInvert_EdgeControl=@n19) {materialFunction=/Game/Effects/Niagara/MasterMaterials/Functions/Common/FXCommon_EdgeMask_Ex}
  n21 = ParticleColor()
  MaterialFunctionCall(Emissive=Multiply(A=LinearInterpolate(A=@n16, B=@n15.RGB, Alpha=@n20.Mask), B=@n21.RGB)) {materialFunction=&MF_FX_EmissiveAdjustments}
  n22 = Multiply(A=@n2.Param4, B=Multiply(A=@n20.Mask, B=ScalarParameter() {defaultValue=0.3, parameterName=c}))
  n23 = TextureObjectParameter() {group=ScanT, parameterName=ScanT_Texture, sortPriority=0, texture=/Game/Weapons/FORT_Melee/Pickaxe_Cube_Ninja_Male/FX/T_Offset_Gradient_01}
  n24 = Constant()
  n25 = VectorParameter() {group=ScanT, parameterName=ScanT_UVControl, channelNames {a=VPan, b=UPan, g=VScale, r=UScale}, defaultValue=<rgba 3 5 0 0>}
  n26 = MaterialFunctionCall!(
   Texture=@n23
   "[Override]DDXInput"=DDX(Value=@n5.ViewportUV)
   "[Override]DDYInput"=DDY(Value=@n5.ViewportUV)
   AdditionalUVOffset=Add(A=@n12, B=MaterialFunctionCall(X=@n24, Y=@n2.Pan) {materialFunction=&MakeFloat2}.Result)
   UVControl_RGB=MaterialFunctionCall(X=@n25.UScale, Y=@n2.ScaleV, Z=@n25.UScale) {materialFunction=&MakeFloat3}.Result
   UVControl_A=@n25.VPan
   "[Override]TimeInput"
   "[Override]UVInput"=@n6.Radial Coordinates
  ) {materialFunction=/Game/Effects/Niagara/MasterMaterials/Functions/Common/FXCommon_TextureSample_DDXDDY}
  n27 = VectorParameter() {parameterName=Refine, defaultValue=<rgba 3 5 1.5 0.5>}
  n28 = Multiply(
   A=MaterialFunctionCall(Texture=@n26.R, RGB=@n27.RGB, A=@n27.RGB) {materialFunction=&MF_FX_Refine}.Result
   B=MaterialFunctionCall!(TexCoord=@n26.UV, ClampChannel=Constant2Vector() {g=1}) {materialFunction=/Game/Effects/Fort_Effects/MaterialFunctions/MF_FX_TilingMethod}.Result
  )
  n29 = Constant2Vector() {g=0.5, r=0.5}
  n30 = ScalarParameter() {parameterName=InerMask}
  n31 = Add(A=Multiply(A=SphereMask(A=@n5.ViewportUV, B=@n29, Hardness=@n30) {attenuationRadius=1}, B=ParticleColor().A))
  n32 = VectorParameter() {group=Mask, parameterName=Mask_EdgeMask_InvertControl, channelNames {a="", b=Power, g=Multiply, r=Offset}, defaultValue=<rgba 1 1 1 1>}
  n33 = VectorParameter() {group=Mask, parameterName=Mask_EdgeMask_Control, channelNames {a="", b=Power, g=Multiply, r=Offset}, defaultValue=<rgba 0 2 2 1>}
  n34 = Multiply(
   A=Multiply(A=@n21.A, B=Multiply(A=@n28, B=@n31))
   B=MaterialFunctionCall(UorV=@n17.G, Invert_EdgeControl=@n32, NoInvert_EdgeControl=@n33) {materialFunction=&FXCommon_EdgeMask_Ex}.Mask
  )
  n35 = CameraPositionWS()
  n36 = DynamicParameter() {paramNames [Scale,X,Y,Z]}
  n37 = MaterialFunctionCall(X=@n36.X, Y=@n36.X, Z=@n36.X) {materialFunction=&MakeFloat3}
  n38 = Divide(
   A=Length(None=Subtract(A=@n35, B=@n37.Result)) {is=class MaterialExpressionLength}
   B=Multiply(A=@n36.Scale, B=ScalarParameter() {defaultValue=50, parameterName=BaseScaleSPhere}) {constB=50}
  )
  n39 = ScalarParameter() {defaultValue=0.4, parameterName=BlendWidth}
  n40 = Saturate(None=Subtract(A=Saturate(None=@n38), B=@n39) {constB=20})
  n41 = Divide(B=OneMinus(None=@n39))
  n42 = Power(Base=OneMinus(None=Multiply(A=@n40, B=@n41))) {constExponent=1.5}
  n43 = MaterialFunctionCall() {materialFunction=&CameraDirectionVector}
  n44 = Normalize(VectorInput=Subtract(A=@n37.Result, B=CameraPositionWS()))
  n45 = VectorParameter() {parameterName=DirectionOffset, defaultValue=<rgba -0.1 2 2 1>}
  n46 = Multiply(A=Add(A=DotProduct(A=@n43.Result, B=@n44), B=@n45.R), B=@n45.G)
  Multiply(A=Multiply(A=Add(A=@n22, B=@n34), B=VertexInterpolator(VS=@n42).PS), B=Power(Base=Saturate(None=@n46), Exp=@n45.B))
 ]
}
